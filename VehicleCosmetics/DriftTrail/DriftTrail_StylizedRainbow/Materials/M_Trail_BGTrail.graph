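# Translucent Niagara material for the stylized rainbow drift trail's background layer.
# Sine-wave-distorted, counter-panning liquid normals serve as a mask blending two rotated, scrolling samples of a rainbow curve texture.
# Dynamic parameters drive scroll speed and blend emissive from particle-tinted luminance to full rainbow, also boosting opacity.
# Opacity combines a tapered U-axis fade, panning and static lightning-texture channels, and particle alpha.
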Material /VehicleCosmetics/DriftTrail/DriftTrail_StylizedRainbow/Materials/M_Trail_BGTrail
graph {
 properties {
  bAutomaticallySetUsageInEditor=false
  bUsedWithNiagaraMeshParticles=true
  bUsedWithNiagaraRibbons=true
  bUsedWithNiagaraSprites=true
  blendMode=BLEND_Translucent
  usedShadingModels=Default Lit
 }
 nodes [
  n1 = ParticleColor()
  n2 = MaterialFunctionCall(Float2=TextureCoordinate() {uTiling=0.25, vTiling=0.25}) {materialFunction=&BreakOutFloat2Components}
  n3 = MaterialFunctionCall(X=Power(Base=@n2.R) {constExponent=1}, Y=@n2.G) {materialFunction=&MakeFloat2}
  n4 = Rotator(Coordinate=@n3.Result, Time=Constant() {r=-0.75}) {centerX=0, centerY=0, speed=3.14}
  n5 = DynamicParameter() {paramNames [RainbowIntensity_01, ReflectionAmount, RainbowSpeed, Color?]}
  n6 = Multiply(A=Time(), B=@n5.RainbowSpeed)
  n7 = Rotator(Coordinate=@n3.Result, Time=Constant() {r=-0.25}) {speed=3.14}
  n8 = TextureCoordinate() {vTiling=0.5}
  n9 = Multiply(A=TextureCoordinate() {uTiling=0.25}, B=Constant2Vector() {g=1, r=4})
  n10 = Frac(None=Rotator(Coordinate=@n9, Time=Constant() {r=1}))
  n11 = Time()
  n12 = Sine(None=Add(A=ComponentMask(None=@n10) {r=true}, B=OneMinus(None=@n11))) {period=0.5}
  n13 = OneMinus(None=Time())
  n14 = Constant() {r=0.5}
  n15 = Sine(None=Add(A=ComponentMask(None=@n10) {r=true}, B=Multiply(A=@n13, B=@n14))) {period=0.25}
  n16 = Add(A=Saturate(None=ConstantBiasScale(None=@n12) {bias=-0.05}), B=Saturate(None=ConstantBiasScale(None=@n15) {bias=-0.075}))
  n17 = Frac(None=Rotator(Coordinate=@n9, Time=Constant() {r=0.85}) {speed=3.14})
  n18 = Time()
  n19 = Sine(None=Add(A=ComponentMask(None=@n17) {r=true}, B=OneMinus(None=@n18))) {period=0.5}
  n20 = Constant() {r=0.25}
  n21 = Sine(None=Add(A=ComponentMask(None=@n17) {r=true}, B=Multiply(A=@n13, B=@n20))) {period=0.25}
  n22 = Add(A=Saturate(None=ConstantBiasScale(None=@n19) {bias=-0.05}), B=Saturate(None=ConstantBiasScale(None=@n21) {bias=-0.085}))
  n23 = Multiply(A=Constant() {r=0.2}, B=AppendVector(A=@n16, B=@n22))
  n24 = TextureSample(UVs=Panner(Coordinate=Add(A=@n8, B=@n23), Time=@n6) {speedX=-0.15}) {samplerType=SAMPLERTYPE_Normal, texture=/DealershipSharedLibrary/FX/Textures/Normal/T_VCVFX_Liquid_Tile_N}
  n25 = TextureCoordinate() {uTiling=-1, vTiling=-0.5}
  n26 = Multiply(A=AppendVector(A=@n22, B=@n16), B=Constant() {r=0.2})
  n27 = TextureSample(UVs=Panner(Coordinate=Add(A=@n25, B=@n26), Time=@n6) {speedX=0.2}) {samplerType=SAMPLERTYPE_Normal, texture=&T_VCVFX_Liquid_Tile_N}
  n28 = Add(A=ConstantBiasScale(None=Add(A=@n24.RGB, B=@n27.RGB)) {bias=1.2}, B=Constant3Vector() {constant=<rgba 0.3 0.1 1 1>})
  n29 = LinearInterpolate(
   A=TextureSample(UVs=Panner(Coordinate=@n4, Time=@n6) {speedX=-0.1, speedY=-0.1}) {texture=/DealershipSharedLibrary/FX/Textures/Diffuse/T_VCVFX_Curve_Rainbow_D}.RGB
   B=TextureSample(UVs=Panner(Coordinate=@n7, Time=@n6) {speedX=-0.1, speedY=-0.1}) {texture=&T_VCVFX_Curve_Rainbow_D}.RGB
   Alpha=Desaturation(None=Saturate(None=@n28)) {luminanceFactors=<rgba 0.3 0.59 0.11 0>}
  )
  n30 = Desaturation(None=@n29) {luminanceFactors=<rgba 0.3 0.59 0.11 0>}
  n31 = Clamp(None=@n30, Min=@n5.Color?) {minDefault=0.6}
  n32 = DynamicParameter() {paramNames [RainbowIntensity_01, ReflectionAmount, RainbowSpeed, Color?]}
  MaterialFunctionCall(Emissive=LinearInterpolate(A=Multiply(A=@n1.RGB, B=@n31), B=@n29, Alpha=@n32.RainbowIntensity_01)) {materialFunction=&MF_FX_EmissiveAdjustments}
  n33 = ComponentMask(None=TextureCoordinate()) {r=true}
  n34 = Power(Base=@n33, Exp=Constant() {r=0.8})
  n35 = Power(Base=OneMinus(None=@n33), Exp=Constant() {r=0.2})
  n36 = TextureSample(UVs=Panner(Coordinate=TextureCoordinate() {uTiling=-2, vTiling=-1}) {speedX=1.5}) {
   samplerSource=SSM_Wrap_WorldGroupSettings
   samplerType=SAMPLERTYPE_LinearColor
   texture=/DealershipSharedLibrary/FX/Textures/TexturePackage/T_VCVFX_StylizedRainbow_Lightning_RGB
  }
  n37 = Multiply(A=Multiply(A=Multiply(A=@n34, B=@n35), B=@n36.B), B=@n1.A)
  n38 = Multiply(A=@n37, B=Constant() {r=2})
  n39 = TextureSample() {
   samplerSource=SSM_Wrap_WorldGroupSettings
   samplerType=SAMPLERTYPE_LinearColor
   texture=&T_VCVFX_StylizedRainbow_Lightning_RGB
  }
  Saturate(None=Multiply(A=LinearInterpolate(A=@n37, B=@n38, Alpha=@n32.RainbowIntensity_01), B=@n39.G))
  Saturate(None=@n30)
 ]
}
