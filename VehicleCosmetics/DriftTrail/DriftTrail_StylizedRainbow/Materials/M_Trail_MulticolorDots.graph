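# Translucent Niagara material for sprites, ribbons and meshes that draws animated rainbow dots for a drift trail cosmetic.
# Time-driven sine waves warp two panning liquid normal samples, whose desaturated sum blends two rotated, scrolling rainbow samples.
# Sphere-gradient textures mask color and opacity, with particle color tinting both.
# Dynamic parameters set scroll speed (RainbowSpeed) and blend emissive from monochrome tint to rainbow plus flare (Painted).
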Material /VehicleCosmetics/DriftTrail/DriftTrail_StylizedRainbow/Materials/M_Trail_MulticolorDots
graph {
 properties {
  bAutomaticallySetUsageInEditor=false
  bUsedWithNiagaraMeshParticles=true
  bUsedWithNiagaraRibbons=true
  bUsedWithNiagaraSprites=true
  blendMode=BLEND_Translucent
  usedShadingModels=Default Lit
 }
 nodes [
  n1 = TextureCoordinate() {uTiling=0.25, vTiling=0.25}
  n2 = DynamicParameter() {paramNames [Painted,Param2,RainbowSpeed,Param4]}
  n3 = Multiply(A=Time(), B=@n2.RainbowSpeed)
  n4 = Panner(Coordinate=Rotator(Coordinate=@n1, Time=Constant() {r=-0.75}) {centerX=0, centerY=0, speed=3.14}, Time=@n3) {speedX=-0.1, speedY=-0.1}
  n5 = Panner(Coordinate=Rotator(Coordinate=@n1, Time=Constant() {r=-0.25}) {speed=3.14}, Time=@n3) {speedX=-0.1, speedY=-0.1}
  n6 = Multiply(A=TextureCoordinate() {uTiling=0.25}, B=Constant2Vector() {g=1, r=4})
  n7 = Frac(None=Rotator(Coordinate=@n6, Time=Constant() {r=1}))
  n8 = Time()
  n9 = Sine(None=Add(A=ComponentMask(None=@n7) {r=true}, B=OneMinus(None=@n8))) {period=0.5}
  n10 = OneMinus(None=Time())
  n11 = Constant() {r=0.5}
  n12 = Sine(None=Add(A=ComponentMask(None=@n7) {r=true}, B=Multiply(A=@n10, B=@n11))) {period=0.25}
  n13 = Add(A=Saturate(None=ConstantBiasScale(None=@n9) {bias=-0.05}), B=Saturate(None=ConstantBiasScale(None=@n12) {bias=-0.075}))
  n14 = Frac(None=Rotator(Coordinate=@n6, Time=Constant() {r=0.85}) {speed=3.14})
  n15 = Time()
  n16 = Sine(None=Add(A=ComponentMask(None=@n14) {r=true}, B=OneMinus(None=@n15))) {period=0.5}
  n17 = Constant() {r=0.25}
  n18 = Sine(None=Add(A=ComponentMask(None=@n14) {r=true}, B=Multiply(A=@n10, B=@n17))) {period=0.25}
  n19 = Add(A=Saturate(None=ConstantBiasScale(None=@n16) {bias=-0.05}), B=Saturate(None=ConstantBiasScale(None=@n18) {bias=-0.085}))
  n20 = Add(A=TextureCoordinate() {vTiling=0.5}, B=Multiply(A=Constant() {r=0.2}, B=AppendVector(A=@n13, B=@n19)))
  n21 = Add(A=TextureCoordinate() {uTiling=-1, vTiling=-0.5}, B=Multiply(A=AppendVector(A=@n19, B=@n13), B=Constant() {r=0.2}))
  n22 = Add(
   A=TextureSample(UVs=Panner(Coordinate=@n20, Time=@n3) {speedX=-0.15}) {samplerType=SAMPLERTYPE_Normal, texture=/DealershipSharedLibrary/FX/Textures/Normal/T_VCVFX_Liquid_Tile_N}.RGB
   B=TextureSample(UVs=Panner(Coordinate=@n21, Time=@n3) {speedX=0.2}) {samplerType=SAMPLERTYPE_Normal, texture=&T_VCVFX_Liquid_Tile_N}.RGB
  )
  n23 = Saturate(None=Add(A=ConstantBiasScale(None=@n22) {bias=0.5}, B=Constant3Vector() {constant=<rgba 0.3 0.1 1 0>}))
  n24 = Multiply(
   A=TextureSampleParameter2D() {parameterName=Texture, texture=/DealershipSharedLibrary/FX/Textures/Gradient_Masks/T_VCVFX_SphereGradient_A}.RGB
   B=Constant() {r=0.5}
  )
  n25 = Multiply(
   A=LinearInterpolate(
    A=TextureSample(UVs=@n4) {texture=/DealershipSharedLibrary/FX/Textures/Diffuse/T_VCVFX_Curve_Rainbow_D}.RGB
    B=TextureSample(UVs=@n5) {texture=&T_VCVFX_Curve_Rainbow_D}.RGB
    Alpha=Desaturation(None=@n23)
   )
   B=@n24
  )
  n26 = Saturate(None=Multiply(A=Constant() {r=2}, B=Desaturation(None=@n25)))
  n27 = ParticleColor()
  n28 = Add(A=@n25, B=Multiply(A=ScalarParameter() {defaultValue=0.1, parameterName=AddFlare}, B=@n24))
  MaterialFunctionCall(Emissive=LinearInterpolate(A=Multiply(A=@n27.RGB, B=@n26), B=Multiply(A=@n27.RGB, B=@n28), Alpha=@n2.Painted)) {materialFunction=&MF_FX_EmissiveAdjustments}
  Saturate(None=Multiply(
    A=@n27.A
    B=TextureSampleParameter2D() {parameterName=Opacity, texture=&T_VCVFX_SphereGradient_A}.RGB
   ))
 ]
}
