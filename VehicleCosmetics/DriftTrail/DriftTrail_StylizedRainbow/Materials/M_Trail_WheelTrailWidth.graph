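# Translucent Niagara material for a stylized rainbow drift trail, rendering a scrolling lightning texture mask scaled by particle alpha.
# Two counter-panned liquid normal samples, warped by rotated sine waves, blend two panning rainbow samples at a dynamic RainbowSpeed.
# The Painted dynamic parameter blends from particle color shaded by clamped rainbow luminance to full rainbow before emissive adjustment, and also boosts the lightning mask.
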Material /VehicleCosmetics/DriftTrail/DriftTrail_StylizedRainbow/Materials/M_Trail_WheelTrailWidth
graph {
 properties {
  bAutomaticallySetUsageInEditor=false
  bUsedWithNiagaraMeshParticles=true
  bUsedWithNiagaraRibbons=true
  bUsedWithNiagaraSprites=true
  blendMode=BLEND_Translucent
  usedShadingModels=Default Lit
 }
 nodes [
  n1 = TextureCoordinate() {uTiling=4}
  n2 = Constant() {r=1}
  n3 = Constant() {r=1.15}
  n4 = TextureSample(UVs=MaterialFunctionCall(
    UVs=Panner(Coordinate=@n1) {speedX=2}
    Texture Scale=MaterialFunctionCall(X=@n2, Y=@n3) {materialFunction=&MakeFloat2}.Result
   ) {materialFunction=&ScaleUVsByCenter}.UVs) {samplerType=SAMPLERTYPE_LinearColor, texture=/DealershipSharedLibrary/FX/Textures/TexturePackage/T_VCVFX_StylizedRainbow_Lightning_RGB}
  MaterialExpressionVertexColor_0 = ParticleColor()
  n5 = Multiply(A=Multiply(A=Constant() {r=1}, B=@n4.B), B=@MaterialExpressionVertexColor_0.A)
  n6 = Constant() {r=2}
  n7 = DynamicParameter() {defaultValue=<rgba 2 10 2 0.6>, paramNames [Painted, Param2, RainbowSpeed, Color?]}
  MaterialExpressionConstantClamp_0 = Saturate(None=LinearInterpolate(A=@n5, B=Multiply(A=@n5, B=@n6), Alpha=@n7.Painted))
  n8 = MaterialFunctionCall(Float2=TextureCoordinate() {uTiling=0.25, vTiling=0.25}) {materialFunction=&BreakOutFloat2Components}
  n9 = MaterialFunctionCall(X=Power(Base=@n8.R) {constExponent=1}, Y=@n8.G) {materialFunction=&MakeFloat2}
  n10 = Rotator(Coordinate=@n9.Result, Time=Constant() {r=-0.75}) {centerX=0, centerY=0, speed=3.14}
  n11 = Multiply(A=Time(), B=@n7.RainbowSpeed)
  n12 = Rotator(Coordinate=@n9.Result, Time=Constant() {r=-0.25}) {speed=3.14}
  n13 = Multiply(A=TextureCoordinate() {uTiling=0.25}, B=Constant2Vector() {g=1, r=4})
  n14 = Frac(None=Rotator(Coordinate=@n13, Time=Constant() {r=1}))
  n15 = Add(A=ComponentMask(None=@n14) {r=true}, B=OneMinus(None=Time()))
  MaterialExpressionConstantClamp_2 = Saturate(None=ConstantBiasScale(None=Sine(None=@n15) {period=0.5}) {bias=-0.05})
  n16 = OneMinus(None=Time())
  n17 = Add(A=ComponentMask(None=@n14) {r=true}, B=Multiply(A=@n16, B=Constant() {r=0.5}))
  MaterialExpressionConstantClamp_3 = Saturate(None=ConstantBiasScale(None=Sine(None=@n17) {period=0.25}) {bias=-0.075})
  n18 = Add(A=@MaterialExpressionConstantClamp_2, B=@MaterialExpressionConstantClamp_3)
  n19 = Frac(None=Rotator(Coordinate=@n13, Time=Constant() {r=0.85}) {speed=3.14})
  n20 = Add(A=ComponentMask(None=@n19) {r=true}, B=OneMinus(None=Time()))
  MaterialExpressionConstantClamp_4 = Saturate(None=ConstantBiasScale(None=Sine(None=@n20) {period=0.5}) {bias=-0.05})
  n21 = Add(A=ComponentMask(None=@n19) {r=true}, B=Multiply(A=@n16, B=Constant() {r=0.25}))
  MaterialExpressionConstantClamp_5 = Saturate(None=ConstantBiasScale(None=Sine(None=@n21) {period=0.25}) {bias=-0.085})
  n22 = Add(A=@MaterialExpressionConstantClamp_4, B=@MaterialExpressionConstantClamp_5)
  n23 = Add(A=TextureCoordinate() {vTiling=0.5}, B=Multiply(A=Constant() {r=0.2}, B=AppendVector(A=@n18, B=@n22)))
  n24 = Add(A=TextureCoordinate() {uTiling=-1, vTiling=-0.5}, B=Multiply(A=AppendVector(A=@n22, B=@n18), B=Constant() {r=0.2}))
  n25 = Add(
   A=TextureSample(UVs=Panner(Coordinate=@n23, Time=@n11) {speedX=-0.15}) {samplerType=SAMPLERTYPE_Normal, texture=/DealershipSharedLibrary/FX/Textures/Normal/T_VCVFX_Liquid_Tile_N}.RGB
   B=TextureSample(UVs=Panner(Coordinate=@n24, Time=@n11) {speedX=0.2}) {samplerType=SAMPLERTYPE_Normal, texture=&T_VCVFX_Liquid_Tile_N}.RGB
  )
  MaterialExpressionConstantClamp_1 = Saturate(None=Add(A=ConstantBiasScale(None=@n25) {bias=0.5}, B=Constant3Vector() {constant=<rgba 0.3 0.1 1 1>}))
  n26 = LinearInterpolate(
   A=TextureSample(UVs=Panner(Coordinate=@n10, Time=@n11) {speedX=-0.1, speedY=-0.1}) {texture=/DealershipSharedLibrary/FX/Textures/Diffuse/T_VCVFX_Curve_Rainbow_D}.RGB
   B=TextureSample(UVs=Panner(Coordinate=@n12, Time=@n11) {speedX=-0.1, speedY=-0.1}) {texture=&T_VCVFX_Curve_Rainbow_D}.RGB
   Alpha=Desaturation(None=@MaterialExpressionConstantClamp_1) {luminanceFactors=<rgba 0.3 0.59 0.11 0>}
  )
  n27 = Desaturation(None=@n26) {luminanceFactors=<rgba 0.3 0.59 0.11 0>}
  MaterialExpressionConstantClamp_6 = Saturate(None=@n27)
  n28 = Clamp(None=@n27, Min=@n7.Color?) {minDefault=0.6}
  MaterialFunctionCall(Emissive=LinearInterpolate(A=Multiply(A=@MaterialExpressionVertexColor_0.RGB, B=@n28), B=@n26, Alpha=@n7.Painted)) {materialFunction=&MF_FX_EmissiveAdjustments}
  n29 = ComponentMask(None=TextureCoordinate()) {r=true}
  Multiply(A=Power(Base=@n29, Exp=Constant() {r=0.8}), B=Power(Base=OneMinus(None=@n29), Exp=Constant() {r=0.2}))
 ]
}
